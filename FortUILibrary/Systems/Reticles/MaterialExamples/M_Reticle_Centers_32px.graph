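# UI example material that draws a reticle center as a dot, ring or cross from signed distance fields.
# A circle, an outlined circle and two crossed boxes form the shapes, and CenterController blends their fields through Lerp_3Color.
# The blended field is anti-aliased at default and shadow-thickness thresholds.
# The two masks mix by shadow opacity, then scale by center opacity.
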Material /FortUILibrary/Systems/Reticles/MaterialExamples/M_Reticle_Centers_32px
graph {
 properties {
  bAutomaticallySetUsageInEditor=false
  bUsedWithStaticMesh=false
  blendMode=BLEND_Translucent
  materialDomain=MD_UI
  usedShadingModels=Default Lit
 }
 nodes [
  n1 = ScalarParameter() {
   defaultValue=0.5
   desc="Controls which center appears. \r\n[0]: Dot\r\n[0.5]: Ring\r\n[1]: Cross"
   group=Center Modifiers
   parameterName=CenterController
   sortPriority=0
  }
  n2 = MaterialFunctionCall(A=NamedRerouteUsage(), B=NamedRerouteUsage(), C=NamedRerouteUsage(), Alpha=@n1) {materialFunction=&Lerp_3Color}
  n3 = ScalarParameter() {defaultValue=0.05, desc=Controls thickness of Drop Shadow, group=Center Modifiers, parameterName=CenterShadowThickness, sliderMax=0.1}
  Multiply(
   A=LinearInterpolate(
    A=MaterialFunctionCall(DistanceField=@n2.Result) {materialFunction=&MF_UI_AntiAliasing}.Inner
    B=MaterialFunctionCall(DistanceField=@n2.Result, Threshold=@n3) {materialFunction=&MF_UI_AntiAliasing}.Inner
    Alpha=ScalarParameter() {defaultValue=1, group=Center Modifiers, parameterName=CenterShadowOpacity, sliderMax=1, sortPriority=2}
   ) {constAlpha=0.4}
   B=ScalarParameter() {defaultValue=0.8, group=Center Modifiers, parameterName=CenterOpacity, sliderMax=1, sortPriority=1}
  )
  NamedRerouteDeclaration(None=MaterialFunctionCall!(UVs, Radius=ScalarParameter() {defaultValue=0.2, group=Dot Modifiers, parameterName=DotSize, sliderMax=0.4, sliderMin=0.16, sortPriority=1}) {materialFunction=/FortUILibrary/MaterialFunctions/Shapes/SDFs/MF_SDF_Circle}.SDF) {name=Dot, nodeColor=<rgba 0 0 0 1>}
  n4 = ScalarParameter() {defaultValue=0.5, group=Cross Modifiers, parameterName=CrossLength, sliderMax=0.9, sliderMin=0.25, sortPriority=4}
  n5 = ScalarParameter() {defaultValue=0.1, group=Cross Modifiers, parameterName=CrossThickness, sliderMax=0.2, sliderMin=0.075, sortPriority=5}
  n6 = AppendVector(A=@n4, B=@n5)
  n7 = AppendVector(A=@n5, B=@n4)
  NamedRerouteDeclaration(None=Min(
    A=MaterialFunctionCall!(UVs, Size=@n6, CornerRoundness, KeepAspectRatio, PixelSize) {materialFunction=/FortUILibrary/MaterialFunctions/Shapes/SDFs/MF_SDF_Box}.SDF
    B=MaterialFunctionCall(Size=@n7) {materialFunction=&MF_SDF_Box}.SDF
   )) {name=Cross, nodeColor=<rgba 0 0 0 1>}
  n8 = ScalarParameter() {defaultValue=0.35, group=Ring Modifiers, parameterName=RingSize, sliderMax=0.8, sliderMin=0.2, sortPriority=2}
  n9 = ScalarParameter() {defaultValue=0.075, group=Ring Modifiers, parameterName=RingThickness, sliderMax=0.15, sliderMin=0.05, sortPriority=3}
  NamedRerouteDeclaration(None=MaterialFunctionCall!(
    DistanceField=MaterialFunctionCall(Radius=@n8) {materialFunction=&MF_SDF_Circle}.SDF
    Threshold
    Thickness=@n9
    Alignment
    Smoothing
   ) {materialFunction=/FortUILibrary/MaterialFunctions/Shapes/Modifiers/MF_UI_ShapeOutline}.SDFOutline) {name=Ring, nodeColor=<rgba 0 0 0 1>}
 ]
}
