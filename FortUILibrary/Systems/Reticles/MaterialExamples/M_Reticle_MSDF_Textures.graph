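# Translucent UI material example of a reticle built from MSDF bracket textures and a procedural center dot.
# MF_UI_Reticle_UVs repeats the brackets by a side count, with rotation and SpreadOffset for accuracy spread.
# Inner bracket UVs subtract that spread to hold the center bars in place, then apply positional offset and scale.
# A Max merges both brackets and the scaled circle SDF, then two anti-aliasing thresholds are blended.
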Material /FortUILibrary/Systems/Reticles/MaterialExamples/M_Reticle_MSDF_Textures
graph {
 properties {
  bAutomaticallySetUsageInEditor=false
  bUsedWithStaticMesh=false
  blendMode=BLEND_Translucent
  materialDomain=MD_UI
  usedShadingModels=Default Lit
 }
 nodes [
  n1 = TextureSample(UVs=NamedRerouteUsage()) {samplerType=SAMPLERTYPE_LinearColor, texture=/FortUILibrary/Systems/Reticles/MaterialExamples/Textures/T_MSDF_OuterBracket}
  n2 = NamedRerouteUsage()
  n3 = ScalarParameter() {desc=Variable for Accuracy Spread Data, group=Reticle Operators, parameterName=SpreadOffset, sliderMax=1, sliderMin=-1}
  n4 = ScalarParameter() {
   defaultValue=-0.175
   desc=Variable to Position Reticle Prongs
   group=Reticle Operators
   parameterName=PositionalOffset
   sliderMax=1
   sliderMin=-1
  }
  n5 = AppendVector(A=Constant(), B=Add(A=Multiply(A=@n3) {constB=-1, desc=Invert}, B=@n4)) {desc=Inverts SpreadOffset to Keep Center Bars In Place}
  n6 = TextureSample(UVs=MaterialFunctionCall(
    UVs=Add(A=@n2, B=@n5) {constB=-0.1, desc=Add Offset}
    Texture Scale=ScalarParameter() {defaultValue=0.625, group=Prong Operators, parameterName=ProngScale, sliderMax=2}
   ) {materialFunction=&ScaleUVsByCenter}.UVs) {samplerType=SAMPLERTYPE_LinearColor, texture=/FortUILibrary/Systems/Reticles/MaterialExamples/Textures/T_MSDF_InnerBracket}
  n7 = Constant() {r=0.055}
  n8 = Max(
   A=Max(
    A=MaterialFunctionCall!(MSDF RGB=@n1.RGB) {materialFunction=/FortUILibrary/MaterialFunctions/Shapes/Modifiers/MF_UI_MSDFtoSDF}.SDF
    B=MaterialFunctionCall(MSDF RGB=@n6.RGB) {materialFunction=&MF_UI_MSDFtoSDF}.SDF
   )
   B=Multiply(A=MaterialFunctionCall!(UVs, Radius=@n7) {materialFunction=/FortUILibrary/MaterialFunctions/Shapes/SDFs/MF_SDF_Circle}.SDF) {constB=-15}
  )
  LinearInterpolate(
   A=MaterialFunctionCall(DistanceField=@n8, Threshold=Constant() {r=0.5}) {materialFunction=&MF_UI_AntiAliasing}.Outer
   B=MaterialFunctionCall(DistanceField=@n8, Threshold=Constant() {r=0.35}) {materialFunction=&MF_UI_AntiAliasing}.Outer
  ) {constAlpha=0.4}
  n9 = ScalarParameter() {desc=Variable for Accuracy Spread Data, group=Reticle Operators, parameterName=SpreadOffset, sliderMax=1, sliderMin=-1}
  NamedRerouteDeclaration(None=MaterialFunctionCall!(
    Reticle UVs
    Reticle Sides=ScalarParameter() {defaultValue=3, group=Reticle Operators, parameterName=ReticleSides, sliderMax=12}
    Reticle Scale
    Reticle Rotation=ScalarParameter() {group=Reticle Operators, parameterName=ReticleRotation, sliderMax=1}
    Prong Scale
    Prong Rotation
    Spread Offset=@n9
    Positional Offset
    "Spread Input Range [Min,Max]"
    "Spread Remapped Range [Min,Max]"
   ) {materialFunction=/FortUILibrary/Systems/Reticles/MaterialFunctions/MF_UI_Reticle_UVs}.UVs) {name=Reticle UVs, nodeColor=<rgba 0 0 0 1>}
 ]
}
